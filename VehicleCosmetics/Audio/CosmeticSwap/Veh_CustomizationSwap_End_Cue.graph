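# Sound cue for the end of a vehicle cosmetic customization swap, played under the vehicles sound class.
# A Random node picks one of two equally weighted end-sound waves and feeds it into a root Modulator left at default settings.
# The Random node is excluded from branch culling, and the cue plays at reduced volume through a dedicated swap attenuation asset.
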SoundCue /VehicleCosmetics/Audio/CosmeticSwap/Veh_CustomizationSwap_End_Cue
root=Modulator_0
nodes [
 {id=Modulator_0, children [Random_1]}
 {id=Random_1, children [WavePlayer_6,WavePlayer_7], properties {bShouldExcludeFromBranchCulling=true, weights [1,1]}}
 {id=WavePlayer_6, properties {soundWaveAssetPtr=&/Vehicle_CustomizationSwap_End_01}}
 {id=WavePlayer_7, properties {soundWaveAssetPtr=&/Vehicle_CustomizationSwap_End_02}}
]
properties {
 attenuationSettings=&/Vehicle_Customization_Swap_Att
 maxDistance=2500
 soundClassObject=&Fort_Vehicles
 volumeMultiplier=0.3
}
