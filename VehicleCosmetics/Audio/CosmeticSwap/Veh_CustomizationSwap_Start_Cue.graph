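# Sound cue for the start of a vehicle customization swap, played in the vehicles sound class with swap-specific attenuation.
# A Mixer layers a one-shot start sound, picked at random from two waves and passed through a Modulator, with a slightly quieter looping swap bed.
# An Enveloper fades the loop in over about a quarter second, and the random node is excluded from branch culling.
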SoundCue /VehicleCosmetics/Audio/CosmeticSwap/Veh_CustomizationSwap_Start_Cue
root=Mixer_0
nodes [
 {
  id=Enveloper_0
  children [WavePlayer_0]
  properties {pitchCurve {editorCurveData {keys [{time=0}]}}, volumeCurve {editorCurveData {keys [<at 0 0.0032>,<at 0.2501 1.0293>]}}}
 }
 {id=Mixer_0, children [Modulator_0,Enveloper_0], properties {inputVolume [1,0.7]}}
 {id=Modulator_0, children [Random_1]}
 {id=Random_1, children [WavePlayer_6,WavePlayer_7], properties {bShouldExcludeFromBranchCulling=true, weights [1,1]}}
 {id=WavePlayer_0, properties {bLooping=true, soundWaveAssetPtr=&/Vehicle_CustomizationSwap_Loop_01}}
 {id=WavePlayer_6, properties {soundWaveAssetPtr=&/Vehicle_CustomizationSwap_Start_02}}
 {id=WavePlayer_7, properties {soundWaveAssetPtr=&/Vehicle_CustomizationSwap_Start_01}}
]
properties {
 attenuationSettings=&/Vehicle_Customization_Swap_Att
 maxDistance=2500
 soundClassObject=&Fort_Vehicles
 volumeMultiplier=0.3
}
